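# Physics asset for the LunarGum two-handed pickaxe mesh, a cosmetic melee weapon.
# It defines a single capsule body on the Melee_TwoHanded_Handle bone, offset and rotated along the weapon as a simple collision volume.
# There are no constraints, so it has no articulated or jointed physics parts.
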PhysicsAsset /BRCosmetics/Weapons/FORT_Melee/Pickaxe_LunarGum/Meshes/LunarGum_Axe_PhysicsAsset
bodies {Melee_TwoHanded_Handle [{shapeType=Capsule, length=71.6344, radius=23.8077, center=<0.7475 -3.7572 33.041>, rotation=<rot 2.7453 -90 -0.0823>}]}
constraints []
